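# Translucent subsurface material for winter-art bubbles, with a faked highlight and soft edges.
# A color ramp over scene depth blends a blue tint toward IceColor from MF_SharedWinter, then darkens or brightens with camera-relative height.
# A tight highlight from the offset camera vector against the vertex normal is added on top.
# An opacity-style term multiplies vertex alpha by inverted Fresnel, fading the silhouette.
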Material /Figment_S08/Athena/WinterArt/Materials/M_Bubbles
graph {
 properties {blendMode=BLEND_Translucent, pixelDepthOffsetMode=PDOM_Legacy, shadingModel=MSM_Subsurface, usedShadingModels=Subsurface}
 nodes [
  n1 = VectorParameter() {parameterName=Param_1, defaultValue=<rgba 0 0 -0.3239 0>}
  n2 = CameraVectorWS()
  n3 = Power(Base=DotProduct(A=Add(A=@n1.RGB, B=@n2), B=VertexNormalWS()), Exp=Constant() {r=12})
  n4 = Constant() {r=0.35}
  n5 = SceneDepth()
  n6 = Constant() {r=2000}
  n7 = LinearInterpolate(
   A=VectorParameter() {defaultValue=<rgba 0.0592 0.1278 0.1823 0>}.RGB
   B=MaterialFunctionCall() {materialFunction=/Game/Athena/WinterArt/Materials/MF_SharedWinter}.IceColor
   Alpha=Saturate(None=Divide(A=@n5.R, B=@n6))
  )
  n8 = Constant3Vector() {constant=<rgba 2.2274 2.5 2.0045 0>}
  n9 = Divide(
   A=ComponentMask(None=WorldPosition() {worldPositionShaderOffset=WPT_CameraRelative}.XYZ) {b=true}
   B=ScalarParameter() {defaultValue=482.9759, parameterName=Param_8}
  )
  Add(
   A=Multiply(A=Saturate(None=@n3), B=Constant() {r=0.5})
   B=LinearInterpolate(A=Multiply(A=@n4, B=@n7), B=Multiply(A=@n7, B=@n8), Alpha=Saturate(None=@n9))
  )
  Constant3Vector() {constant=<rgba 0.0126 0.08 0.1615 0>}
  n10 = Fresnel(ExponentIn=Constant() {r=1}) {baseReflectFraction=0, exponent=1}
  Multiply(A=VertexColor().A, B=Multiply(A=OneMinus(None=@n10), B=Constant() {r=0.75}))
 ]
}
